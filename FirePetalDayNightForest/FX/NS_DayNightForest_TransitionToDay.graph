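# Niagara one-shot burst effect for the day/night forest's transition to day, built from warm gold and orange layers.
# An expanding energy sphere mesh, ray meshes and optical flare sprites form the flash, while Glisten and GPU NOTILT sprites scatter and drift with curl noise.
# Distance remaps adjust flash scale and camera offset, and a front-end modifier adjusts sphere growth.
# Rays also animates PPM blend values alongside a component renderer, weighted by distance.
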NiagaraSystem /FirePetalDayNightForest/FX/NS_DayNightForest_TransitionToDay
dependencies {
 usedDataInterfaces [&NiagaraDataInterfaceColorCurve, &NiagaraDataInterfaceCurve, /Script/Niagara.NiagaraDataInterfaceOcclusion]
 usedDynamicInputs []
 usedModules [&SystemState]
}
emitters {
 Energy_Sphere {
  emitterUpdateScript [
   &EmitterState()
   &SpawnBurst_Instantaneous(Spawn Count=1, Use Spawn Probability=true)
  ]
  particleSpawnScript [&InitializeParticle(
    Color=<rgba 5 3.8387 1.4816 1>
    Lifetime=0.75
    Mesh Scale Mode=::NewEnumerator3
    Mesh Uniform Scale=15
    Position Offset=<0 0 100>
    UsePositionOffset=true
    Write Material Random=true
   ) {module=InitializeParticle001}]
  particleUpdateScript [
   &ParticleState()
   &ScaleColor(
    Scale Alpha=&FloatFromCurve(FloatCurve=DataInterface() {propertyValues {Curve {keys [{arriveTangent=-2.4108, interpMode=RCIM_Cubic, leaveTangent=-2.4108, tangentMode=RCTM_User, time=0.3, value=1}, <at 1 0 RCIM_Cubic>]}}})
    ScaleA=true
    ScaleRGB=true
   )
   &ScaleMeshSize(Scale Factor=&VectorFromFloat(Value=&FloatFromCurve(
      FloatCurve=DataInterface() {propertyValues {Curve {keys [
          {arriveTangent=3.1584, interpMode=RCIM_Cubic, leaveTangent=3.1584, tangentMode=RCTM_User, time=0, value=0}
          {arriveTangent=0.1704, interpMode=RCIM_Cubic, leaveTangent=0.1704, tangentMode=RCTM_User, time=1, value=1.9}
         ]}}}
      Scale Curve=&FN_FrontEndFloatModifier(Gameplay=1, Front End=0.78)
     )))
   &ScaleMeshSize(Scale Factor=&VectorFromFloat(Value=&Lerp_Float(
      A=0.01
      B=1.5
      Alpha=&FN_LOD_Distance_Remap(Start Distance=0, End Distance=2500, Minimum Output=0, Maximum Output=1, Clamp Output=true)
     ))) {module=ScaleMeshSize001}
   &DynamicMaterialParameters(
    Index 0 Param 1=1
    Index 0 Param 2=&FloatFromCurve(FloatCurve=DataInterface() {propertyValues {Curve {keys [<at 0 1>,<at 1 0>]}}})
    Module.Index0Param0Write=true
    Module.Index0Param1Write=true
    Module.Index0Param2Write=true
    Module.Index0Param3Write=true
   )
  ]
  rendererClasses [&NiagaraMeshRendererProperties]
  renderers [{rendererClass=&NiagaraMeshRendererProperties, rendererIndex=0}]
 }
 Glisten {
  emitterUpdateScript [
   &EmitterState()
   &SpawnRate(SpawnGroup=1, SpawnRate=30, Use Spawn Probability=false) {enabled=false}
   &SpawnBurst_Instantaneous(Spawn Count=150, Use Spawn Probability=false, UseLoopCountLimit=false)
  ]
  particleSpawnScript [
   &InitializeParticle(
    Color=<rgba 1 0.7108 0.2724 1>
    Lifetime Max=6
    Lifetime Min=3
    Lifetime Mode=::NewEnumerator1
    Sprite Size Mode=::NewEnumerator1
    Uniform Sprite Size Max=60
    Uniform Sprite Size Min=15
    UsePositionOffset=true
    Write Material Random=true
   )
   &ShapeLocation(MaskSpawnGroup=false, Override Seed=false, Sphere Radius=50, UseOwnerScale=true)
   &AddVelocity(
    Module.Use Falloff Distance=false
    Override Seed=false
    Velocity Mode=::NewEnumerator1
    Velocity Speed=&MaskFloatBySpawnGroup(Masked Float=&RandomRangeFloat(Minimum=350, Maximum=4500), Passthrough Float=50)
   )
   &DynamicMaterialParameters(
    Index 0 Param 1=&RandomRangeFloat(Minimum=0.25, Maximum=0.5)
    Index 0 Param 2=3
    Module.Index0Param0Write=true
    Module.Index0Param1Write=true
    Module.Index0Param2Write=true
    Module.Index0Param3Write=true
   )
  ]
  particleUpdateScript [
   &GravityForce(Gravity=&RandomRangeVector(Minimum=<0 0 -50>, Maximum=<0 0 -250>))
   &CurlNoiseForce(Bias Noise=false, Noise Frequency=1, Noise Strength=350, Pan Noise=false)
   &Drag(Drag=5, Use Linear Drag=true, Use Rotational Drag=true)
   &SolveForcesAndVelocity(Clamp Velocity=false, Limit Acceleration=false)
   &ParticleState()
   &ScaleSpriteSize(Uniform Curve Sprite Scale=DataInterface() {propertyValues {Curve {keys [
        {arriveTangent=3.3457, interpMode=RCIM_Cubic, leaveTangent=3.3457, tangentMode=RCTM_User, time=0, value=0}
        {tangentMode=RCTM_User, time=1, value=1}
       ]}}})
   &ScaleColor(Scale Mode=::NewEnumerator2)
  ]
  rendererClasses [&NiagaraSpriteRendererProperties]
  renderers [{rendererClass=&NiagaraSpriteRendererProperties}]
 }
 NOTILT {
  simTarget=GPUComputeSim
  emitterUpdateScript [
   &EmitterState()
   &SpawnBurst_Instantaneous(Spawn Count=375, Use Spawn Probability=false, UseLoopCountLimit=false)
  ]
  particleSpawnScript [
   &InitializeParticle(
    Color=&ColorFromCurve(
     ColorCurve=DataInterface() {propertyValues {
       AlphaCurve {keys [<at 0.4783 0.25 RCIM_Cubic>]}
       BlueCurve {keys [<at 0 0>,<at 0.3901 0>,<at 0.7425 0>,<at 1 0.4333 RCIM_Cubic>]}
       GreenCurve {keys [<at 0 0.3887>,<at 0.3901 0.9932>,<at 0.7425 0.2687>,<at 1 0.6673 RCIM_Cubic>]}
       RedCurve {keys [<at 0 1>,<at 0.3901 1>,<at 0.7425 1>,<at 1 1 RCIM_Cubic>]}
      }}
     CurveIndex=&RandomRangeFloat(Minimum=0, Maximum=1)
    )
    Lifetime Max=6
    Lifetime Min=2
    Lifetime Mode=::NewEnumerator1
    Sprite Rotation Mode=::NewEnumerator2
    Sprite Size Mode=::NewEnumerator1
    Sprite UV Mode=::NewEnumerator3
    UsePositionOffset=false
    Write Material Random=true
   )
   &ShapeLocation(MaskSpawnGroup=false, Offset Mode=::NewEnumerator0, Override Seed=false, Sphere Radius=1, UseOwnerScale=true)
   &AddVelocity(
    Module.Use Falloff Distance=false
    Override Seed=false
    Velocity Mode=::NewEnumerator1
    Velocity Speed=&RandomRangeFloat(Minimum=500, Maximum=3750)
   )
   &DynamicMaterialParameters(
    Index 0 Param 1=&RandomRangeFloat(Minimum=0.1, Maximum=12)
    Module.Index0Param0Write=true
    Module.Index0Param1Write=true
    Module.Index0Param2Write=true
    Module.Index0Param3Write=true
   )
   &ApplyInitialForces()
  ]
  particleUpdateScript [
   &ParticleState()
   &ScaleSpriteSize(Scale Factor=&Vector2DFromFloat(Value=&FloatFromCurve(FloatCurve=DataInterface() {propertyValues {Curve {keys [
          <at 0 0 RCIM_Cubic>
          <at 0.112 1.0173>
          {arriveTangent=-1.142, interpMode=RCIM_Cubic, leaveTangent=-1.142, time=0.7001, value=0.872}
          <at 1.0011 0.002>
         ]}}}))) {enabled=false}
   &ScaleColor(
    Scale Alpha=&FloatFromCurve(
     FloatCurve=DataInterface() {propertyValues {Curve {keys [<at 0 0.9 RCIM_Cubic>, {arriveTangent=-0.9, interpMode=RCIM_Cubic, leaveTangent=-0.9, time=0.5, value=1}, <at 1 0 RCIM_Cubic>]}}}
     Scale Curve=&Multiply_Float(
      A=1
      B=&Waveform(
       Global Input Position=&RandomRangeFloat(Minimum=0, Maximum=1)
       Global Clamp Min=0.5
       "Waveform [1]"=::NewEnumerator9
       "[1] Frequency"=5
       "[1] Amplitude Min/Max"=<0 1>
       "[1] Scale"=0.5
       "[1] Bias"=1
      )
     )
    )
    ScaleA=true
    ScaleRGB=true
   )
   &CurlNoiseForce(Bias Noise=false, Noise Frequency=1, Noise Strength=150, Pan Noise=false) {module=CurlNoiseForce001}
   &Drag(Drag=3, Use Linear Drag=true, Use Rotational Drag=true)
   &AccelerationForce(Acceleration=<75 0 0>)
   &SolveForcesAndVelocity(Clamp Velocity=false, Limit Acceleration=false)
  ]
  rendererClasses [&NiagaraSpriteRendererProperties]
  renderers [{rendererClass=&NiagaraSpriteRendererProperties}]
 }
 Optical_Slow {
  emitterUpdateScript [
   &EmitterState()
   &SpawnBurst_Instantaneous(Spawn Group=1, Use Spawn Probability=false, UseLoopCountLimit=false) {module=SpawnBurst_Instantaneous001}
   &SpawnBurst_Instantaneous(Use Spawn Probability=false, UseLoopCountLimit=false)
  ]
  particleSpawnScript [&InitializeParticle(
    Color=<rgba 5 2.8735 0.665 1>
    Lifetime=0.75
    Position Offset=<0 0 120>
    Sprite Size=&MaskVector2dBySpawnGroup(Masked Vector2D=<500 500>, Passthrough Vector2D=<1125 25>, Spawn Group Mask=0)
    Sprite Size Mode=::NewEnumerator4
    UsePositionOffset=true
    Write Material Random=true
   )]
  particleUpdateScript [
   &CameraOffset(Camera Offset Amount=&Lerp_Float(
     A=50
     B=1500
     Alpha=&FN_LOD_Distance_Remap(Start Distance=1250, End Distance=2500, Minimum Output=0, Maximum Output=1, Clamp Output=true)
    ))
   &ParticleState()
   &ScaleSpriteSize(Uniform Curve Sprite Scale=DataInterface() {propertyValues {Curve {keys [
        {interpMode=RCIM_Cubic, leaveTangent=21.1632, tangentMode=RCTM_Break, time=0, value=0}
        {arriveTangent=-0.5534, interpMode=RCIM_Cubic, leaveTangent=-0.5534, tangentMode=RCTM_User, time=0.1, value=1.1}
        {arriveTangent=-0.5, interpMode=RCIM_Cubic, leaveTangent=-0.5, time=0.2, value=0.95}
        {arriveTangent=-0.2843, interpMode=RCIM_Cubic, leaveTangent=-0.2843, tangentMode=RCTM_User, time=0.5, value=0.9}
        <at 1 0>
       ]}}})
   &ScaleSpriteSize(
    Scale Sprite Size Mode=::NewEnumerator0
    Uniform Scale Factor=&FN_LOD_Distance_Remap(Start Distance=0, End Distance=2500, Minimum Output=0.5, Maximum Output=1, Clamp Output=true)
   ) {module=ScaleSpriteSize001}
   &ScaleColor(Scale Mode=::NewEnumerator2)
  ]
  rendererClasses [&NiagaraSpriteRendererProperties]
  renderers [{rendererClass=&NiagaraSpriteRendererProperties}]
 }
 Rays {
  emitterUpdateScript [
   &EmitterState()
   &SpawnBurst_Instantaneous(Use Spawn Probability=false, UseLoopCountLimit=false)
  ]
  particleSpawnScript [
   &InitializeParticle(
    Color=<rgba 1 0.7446 0.4554 1>
    Lifetime=1.5
    Mesh Scale=<50 5 5>
    Mesh Scale Mode=::NewEnumerator4
    Position Offset=<0 0 120>
    UsePositionOffset=true
    Write Material Random=true
   )
   &DynamicMaterialParameters(
    Index 0 Param 1=&RandomRangeFloat(Minimum=0, Maximum=0.1)
    Index 0 Param 2=&RandomRangeFloat(Minimum=0.5, Maximum=5)
    Index 0 Param 3=&RandomRangeFloat(Minimum=2, Maximum=2)
    Module.Index0Param0Write=true
    Module.Index0Param1Write=true
    Module.Index0Param2Write=true
    Module.Index0Param3Write=false
   )
   /SetVariables() {isSetParameters=true}
  ]
  particleUpdateScript [
   &ParticleState()
   &ScaleMeshSize(Scale Factor=&VectorFromFloat(Value=&Lerp_Float(
      A=0.1
      B=1.25
      Alpha=&FN_LOD_Distance_Remap(Start Distance=0, End Distance=2500, Minimum Output=0, Maximum Output=1, Clamp Output=true)
     ))) {module=ScaleMeshSize001}
   &ScaleMeshSize(Scale Factor=&MakeVector(
     X=&FloatFromCurve(FloatCurve=DataInterface() {propertyValues {Curve {keys [
          {arriveTangent=8.2778, interpMode=RCIM_Cubic, leaveTangent=8.2778, tangentMode=RCTM_User, time=0, value=0}
          {arriveTangent=-0.0496, interpMode=RCIM_Cubic, leaveTangent=-0.0496, tangentMode=RCTM_User, time=0.3, value=1.25}
          {arriveTangent=-0.8719, interpMode=RCIM_Cubic, leaveTangent=-0.8719, tangentMode=RCTM_User, time=0.6, value=0.125}
          {interpMode=RCIM_Cubic, leaveTangent=-0.1625, time=1, value=0}
         ]}}})
     Y=&FloatFromCurve(FloatCurve=DataInterface() {propertyValues {Curve {keys [
          {arriveTangent=34.3634, interpMode=RCIM_Cubic, time=0, value=0}
          {arriveTangent=1, interpMode=RCIM_Cubic, leaveTangent=1, time=0.05, value=1}
          {arriveTangent=-0.9474, interpMode=RCIM_Cubic, leaveTangent=-0.9474, time=0.5, value=0.5}
          {interpMode=RCIM_Cubic, leaveTangent=-0.1625, time=1, value=0.1}
         ]}}})
     Z=&FloatFromCurve(FloatCurve=DataInterface() {propertyValues {Curve {keys [
          {arriveTangent=34.3634, interpMode=RCIM_Cubic, time=0, value=0}
          {arriveTangent=1, interpMode=RCIM_Cubic, leaveTangent=1, time=0.05, value=1}
          {arriveTangent=-0.9474, interpMode=RCIM_Cubic, leaveTangent=-0.9474, time=0.5, value=0.5}
          {interpMode=RCIM_Cubic, leaveTangent=-0.1625, time=1, value=0.1}
         ]}}})
    ))
   &ScaleColor(
    Scale Alpha=&FloatFromCurve(FloatCurve=DataInterface() {propertyValues {Curve {keys [
         <at 0 0 RCIM_Cubic>
         {arriveTangent=1.9, interpMode=RCIM_Cubic, leaveTangent=1.9, time=0.15, value=1}
         {arriveTangent=-2.4898, interpMode=RCIM_Cubic, leaveTangent=-2.4898, tangentMode=RCTM_User, time=0.5, value=0.95}
         <at 0.7 0 RCIM_Cubic>
        ]}}})
    ScaleA=true
    ScaleRGB=true
   )
   &CameraOffset()
   &self:OcclusionCheck(NumberOfRings=8, SamplesPerRing=5, WorldDiameter=250) {enabled=false, module=OcclusionCheck}
   /SetVariables(
    Particles.DisableByDistance=true
    Particles.PPM_BlendRadius:NiagaraFloat=&FloatFromCurve(FloatCurve=DataInterface() {propertyValues {Curve {keys [
         {interpMode=RCIM_Cubic, leaveTangent=5.1202, tangentMode=RCTM_Break, time=0, value=0}
         {interpMode=RCIM_Cubic, tangentMode=RCTM_Break, time=0.5, value=0.95}
         <at 1 0>
        ]}}})
    Particles.PPM_BlendWeight:NiagaraFloat=&FloatFromCurve(
     FloatCurve=DataInterface() {propertyValues {Curve {keys [
         {arriveTangent=9.1434, interpMode=RCIM_Cubic, leaveTangent=9.1434, tangentMode=RCTM_User, time=0, value=0}
         {interpMode=RCIM_Cubic, tangentMode=RCTM_User, time=0.15, value=1.05}
         <at 0.5 0>
        ]}}}
     Scale Curve=&FN_LOD_Distance_Remap(Start Distance=0, End Distance=1500, Minimum Output=1, Maximum Output=0, Clamp Output=false)
    )
    Particles.PPM_Scale:Vector3f=<1 1 1>
   ) {isSetParameters=true}
  ]
  rendererClasses [&NiagaraMeshRendererProperties, &NiagaraComponentRendererProperties]
  renderers [
   {rendererClass=&NiagaraMeshRendererProperties, rendererIndex=0}
   {rendererClass=&NiagaraComponentRendererProperties, rendererIndex=1}
  ]
 }
}
summary {
 systemName=NS_DayNightForest_TransitionToDay
 emitters [
  {emitterName=NOTILT, simTarget=GPUComputeSim, rendererClasses [&NiagaraSpriteRendererProperties]}
  {emitterName=Rays, rendererClasses [&NiagaraMeshRendererProperties, &NiagaraComponentRendererProperties]}
  {emitterName=Optical_Slow, rendererClasses [&NiagaraSpriteRendererProperties]}
  {emitterName=Energy_Sphere, rendererClasses [&NiagaraMeshRendererProperties]}
  {emitterName=Glisten, rendererClasses [&NiagaraSpriteRendererProperties]}
 ]
 userVariables [
  User.bIsFrontEnd {type=NiagaraBool}
  User.bIsFrontEndPreview {type=NiagaraBool}
 ]
}
